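# Sound cue for the impact sound of the male CerealBox cosmetic pickaxe.
# A mixer layers a boosted, modulated random pick of four impact waves with an occasional vocal layer.
# The vocal branch picks among three pitch-varied Vox waves, behind an equal-weight random node with three empty slots, so it plays about one hit in four.
# It uses shared pickaxe impact attenuation and owner-limited concurrency that stops the oldest sound.
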SoundCue /BRCosmetics/Sounds/PickAxes/CerealBox/Pickaxe_CerealBoxMale_Impact_Cue
root=Mixer_0
nodes [
 {id=Mixer_0, children [Modulator_0,Random_2], properties {inputVolume [1.5,1]}}
 {id=Modulator_0, children [Random_0]}
 {id=Modulator_1, children [Random_1], properties {pitchMax=1.1, pitchMin=0.9}}
 {id=Random_0, children [WavePlayer_1,WavePlayer_2,WavePlayer_3,WavePlayer_4], properties {weights [1,1,1,1]}}
 {id=Random_1, children [WavePlayer_5,WavePlayer_6,WavePlayer_7], properties {weights [1,1,1]}}
 {id=Random_2, children [Modulator_1,null,null,null], properties {weights [1,1,1,1]}}
 {id=WavePlayer_1, properties {soundWaveAssetPtr=&/Pickaxe_CerealBoxMale_Impact_01}}
 {id=WavePlayer_2, properties {soundWaveAssetPtr=&/Pickaxe_CerealBoxMale_Impact_02}}
 {id=WavePlayer_3, properties {soundWaveAssetPtr=&/Pickaxe_CerealBoxMale_Impact_03}}
 {id=WavePlayer_4, properties {soundWaveAssetPtr=&/Pickaxe_CerealBoxMale_Impact_04}}
 {id=WavePlayer_5, properties {soundWaveAssetPtr=&/Pickaxe_CerealBoxMale_Vox_01}}
 {id=WavePlayer_6, properties {soundWaveAssetPtr=&/Pickaxe_CerealBoxMale_Vox_02}}
 {id=WavePlayer_7, properties {soundWaveAssetPtr=&/Pickaxe_CerealBoxMale_Vox_03}}
]
properties {
 attenuationSettings=/Game/Sounds/Attenuation/Fort_Pickaxe_Impact_Att
 audioPropertiesSheet=/BRCosmetics/Sounds/Settings/PropertySheets/Pickaxe/APS_Cosmetic_Pickaxe_Impact
 bOverrideConcurrency=true
 maxDistance=4300
 soundClassObject=&Fort_Impacts
 volumeMultiplier=0.6
 concurrencyOverrides {bLimitToOwner=true, maxCount=4, resolutionRule=StopOldest}
}
